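# Sound cue for a Season 22 Battle Pass voice line, Q6 outro line 010, spoken as the TheNothing_DV distant voice.
# A single dialogue player plays the matching dialogue wave with that speaker as context and no targets.
# The cue uses the Chromehands attenuation asset and routes through the narrator VO sound class.
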SoundCue /BattlePassS22/Audio/VO/Cues/Chromehands/DISTANTVOICE_S22_REGS_Q6_Outro_010_Dialogue_Cue
root=DialoguePlayer_0
nodes [{
  id=DialoguePlayer_0
  properties {dialogueWaveParameter {
    dialogueWave=/BattlePassS22/Audio/VO/Waves/TheNothing/DISTANTVOICE_S22_REGS_Q6_Outro_010_Dialogue
    context {speaker=/BattlePassS22/Audio/VO/TheNothing_DV, targets []}
   }}
 }]
properties {
 attenuationSettings=/BattlePassS22/Audio/VO/S22_Chromehands_Att
 maxDistance=4000
 soundClassObject=/Game/Sounds/FortSoundClassesAndModes/Fort_VO_Narrator
 volumeMultiplier=1
}
